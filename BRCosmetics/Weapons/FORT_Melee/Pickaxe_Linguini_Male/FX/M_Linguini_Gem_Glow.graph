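# Translucent glow material for Niagara mesh particles, named for the gem on the Linguini pickaxe.
# Two panning textures, electric noise and shield caustics, multiply into an animated pattern.
# A camera-to-normal dot product, shaped by power or a cheaper contrast lerp via static switch, masks the pattern toward camera-facing areas.
# Particle color tints the emissive through an adjustment function, while particle alpha scales opacity, with switchable depth fade.
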Material /BRCosmetics/Weapons/FORT_Melee/Pickaxe_Linguini_Male/FX/M_Linguini_Gem_Glow
graph {
 properties {bUsedWithNiagaraMeshParticles=true, blendMode=BLEND_Translucent, pixelDepthOffsetMode=PDOM_Legacy, usedShadingModels=Default Lit}
 nodes [
  n1 = Multiply(
   A=TextureSample(UVs=Panner() {speedX=-0.5, speedY=-0.1}) {samplerType=SAMPLERTYPE_LinearColor, texture=/Game/Characters/Player/Male/Medium/Bodies/M_MED_SRDrift/Textures/T_ElectricNoise}.R
   B=TextureSample(UVs=Panner() {speedX=0.3, speedY=0.2}) {texture=&T_ShieldCaustics}.R
  )
  LinearInterpolate(Alpha=@n1) {constA=0.8, constB=2}
  n2 = ParticleColor()
  n3 = Clamp(None=DotProduct(A=CameraVectorWS(), B=PixelNormalWS()))
  n4 = Multiply(A=ScalarParameter() {defaultValue=1, parameterName=FresnelBrightness})
  n5 = ScalarParameter() {defaultValue=1, parameterName=FresnelPower}
  n6 = StaticSwitchParameter(
   True=LinearInterpolate(A=ScalarParameter() {parameterName=FresnelContrastLow}, B=ScalarParameter() {defaultValue=1, parameterName=FresnelContrastHi}, Alpha=@n3)
   False=Power(Base=Multiply(A=@n3, B=@n4), Exp=Multiply(A=@n5))
  ) {parameterName=FresnelCheapContrast}
  n7 = Clamp(None=@n6)
  MaterialFunctionCall(Emissive=Multiply(A=@n2.RGB, B=Multiply(A=@n1, B=@n7))) {materialFunction=&MF_FX_EmissiveAdjustments}
  OneMinus(None=@n7)
  OneMinus(None=@n6)
  n8 = Multiply(A=Multiply(A=@n7, B=Add(A=@n1) {constB=0.1}), B=@n2.A)
  StaticSwitchParameter(True=DepthFade(Opacity=@n8, FadeDistance=ScalarParameter() {defaultValue=100, parameterName=FadeDistance}), False=@n8) {parameterName=DepthFade}
 ]
}
